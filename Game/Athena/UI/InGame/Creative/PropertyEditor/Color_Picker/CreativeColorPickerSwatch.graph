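# Color swatch widget for the Creative property editor's color picker, showing one tinted option as a swatch image.
# SetColorToRepresent tints the swatch and uses the option's icon, falling back to a stroked texture for pure white or the default texture otherwise.
# SetSelected toggles a selection shadow, hidden on construct.
# SetRaised acts only on state change, lifting or lowering the swatch via a forward/reverse animation or an instant render transform.
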WidgetBlueprint /Game/Athena/UI/InGame/Creative/PropertyEditor/Color_Picker/CreativeColorPickerSwatch
animations [Anim_RaiseSwatch {
  binding_count=1
  length_seconds=1
  object=&self:Anim_RaiseSwatch
  sequence {bindings [Image_Swatch {
     id=#1
     possessed_class=Image
     tracks [{
       type=MovieScene2DTransformTrack
       sections [{
         type=MovieScene2DTransformSection
         end_frame=20
         start_frame=0
         channels [Translation.Y {type=MovieSceneScriptingFloatChannel, keys [[0,0],[19.9997,-8]]}]
        }]
      }]
    }]}
 }]
graphs [
 EventGraph {nodes [
   Event Construct()
   > SetVisibility(self=Get Image_SelectedShadow() -> (Image_SelectedShadow:Image Object Reference), InVisibility=Hidden)
  ]}
 SetColorToRepresent {nodes [
   n1 = SetColorToRepresent!() -> (then, ColorOption:Playlist Option Tinted Icon Int Value Structure) {is=title K2Node_FunctionEntry}
   n2 = Break Playlist Option Tinted Icon Int Value!(PlaylistOptionTintedIconIntValue:"Playlist Option Tinted Icon Int Value Structure (by ref)"=@n1) -> (
    Value:Linear Color Structure
    Icon:Texture 2D Object Reference
    SecondaryValue:int
    DisplayName:Text
    OptionValueName:String
    GrantedContextTags:Gameplay Tag Container Structure
   ) {is=title K2Node_BreakStruct}
   > SetColorAndOpacity(self:Image Object Reference=Get Image_Swatch() -> (Image_Swatch:Image Object Reference), InColorAndOpacity:Linear Color Structure=@n2)
   n3 = Should Have Stroke!(Option:Playlist Option Tinted Icon Int Value Structure=Get ColorOption() -> (ColorOption:Playlist Option Tinted Icon Int Value Structure)) -> (Result:bool)
   n4 = Get SwatchTextureWithStroke() -> (SwatchTextureWithStroke:Texture 2D Object Reference)
   n5 = Get DefaultSwatchTexture() -> (DefaultSwatchTexture:Texture 2D Object Reference)
   > SetBrushFromTexture!(
    execute
    self:Image Object Reference=Get Image_Swatch()
    Texture:Texture 2D Object Reference=Select(
     Option 0:Texture 2D Object Reference=Select(Option 0=@n5, Option 1=@n4, Index=@n3)
     Option 1:Texture 2D Object Reference=@n2.Icon
     Index=IsValid(Object=@n2.Icon)
    ) -> (ReturnValue:Texture 2D Object Reference)
    bMatchSize:bool=true
   ) -> (then)
  ]}
 SetRaised {nodes [
   n1 = SetRaised!() -> (then, bIsRaised:bool=~, bPlayAnimation:bool=~) {is=title K2Node_FunctionEntry}
   > Branch(Condition="Not Equal (Boolean)"!(A:bool=@n1, B:bool=Get bIsSwatchRaised() -> (bIsSwatchRaised:bool)) -> (ReturnValue:bool) {is=title K2Node_CallFunction})
   > Set bIsSwatchRaised(bIsSwatchRaised:bool=Get bIsRaised() -> (bIsRaised:bool)) -> (Output_Get:bool)
   > n2 = Branch(Condition=@n1.bPlayAnimation)
   > PlayAnimation(
    InAnimation=Get Anim_RaiseSwatch() -> (Anim_RaiseSwatch:Widget Animation Object Reference)
    PlayMode=Select(Option 0:EUMGSequencePlayMode Enum=Reverse, Option 1:EUMGSequencePlayMode Enum, Index=Get bIsSwatchRaised()) -> (ReturnValue:EUMGSequencePlayMode Enum)
    PlaybackSpeed=4
   )
   n3 = Get bIsSwatchRaised()
   SetRenderTransform!(
    execute=@n2.else
    self:Widget Object Reference=Get Image_Swatch()
    InTransform:Widget Transform Structure=Make Widget Transform!(
     Translation_X:double=0
     Translation_Y:double=SelectFloat(A=-8, B=0, bPickA=@n3)
     Scale:Vector 2D Structure=(X=1,Y=1)
     Shear:Vector 2D Structure=(X=0,Y=0)
     Angle:float=0
    ) -> (WidgetTransform:Widget Transform Structure) {is=title K2Node_MakeStruct}
   ) -> (then)
  ]}
 SetSelected {nodes [
   n1 = SetSelected!() -> (then, bIsSelected:bool=~) {is=title K2Node_FunctionEntry}
   > SetVisibility(
    self=Get Image_SelectedShadow()
    InVisibility=Select(Option 0:ESlateVisibility Enum=Hidden, Option 1:ESlateVisibility Enum=Visible, Index=@n1) -> (ReturnValue:ESlateVisibility Enum)
   )
  ]}
 Should Have Stroke {nodes [
   n1 = Should Have Stroke!() -> (
    then
    Option_Value:Linear Color Structure
    Option_Icon:Texture 2D Object Reference
    Option_SecondaryValue:int
    Option_DisplayName:Text
    Option_OptionValueName:String
    Option_GrantedContextTags:Gameplay Tag Container Structure
   ) {is=title K2Node_FunctionEntry}
   n2 = IsValid(Object=@n1.Option_Icon)
   > Return Node!(
    execute
    Result:bool=AND Boolean(
     A="Equal (LinearColor)"(A:Linear Color Structure=@n1, B:Linear Color Structure=(R=1,G=1,B=1,A=1)) -> (ReturnValue:bool=~) {is=title K2Node_PromotableOperator}
     B=NOT Boolean(A=@n2)
    )
   )
  ]}
]
variables [
 bIsSwatchRaised {type=boolean}
 DefaultSwatchTexture {type=Texture2D, default=&/T_Icon_ColorPicker_swatch}
 SwatchTextureWithStroke {type=Texture2D, default=&/T_Icon_ColorPicker_swatch_stroke}
]
widgets {
 info {inheritedWidgetCount=0, parentClass=&UserWidget, rootWidgetClass=&Overlay, widgetCount=3}
 widgets [
  Overlay_98 {widgetClassPath=&Overlay, widget=Overlay_98, properties {visibility=HitTestInvisible}}
  Image_SelectedShadow {
   widgetClassPath=&Image
   bIsVariable=true
   parent=Overlay_98
   widget=Image_SelectedShadow
   properties {
    visibility=SelfHitTestInvisible
    brush {
     bIsDynamicallyLoaded=false
     drawAs=Image
     imageType=NoImage
     mirroring=NoMirror
     resourceName=None
     resourceObject=&/T_Icon_ColorPicker_swatch_selected
     tiling=NoTile
     imageSize=<61 31>
     margin=<ltrb 0 0 0 0>
     outlineSettings {
      bUseBrushTransparency=false
      roundingType=HalfHeightRadius
      width=0
      color {colorUseRule=UseColor_Specified, specifiedColor=<rgba 0 0 0 0>}
      cornerRadii=<0 0 0 0>
     }
     tintColor {colorUseRule=UseColor_Specified, specifiedColor=<rgba 1 1 1 1>}
     uVRegion {bIsValid=false, max=<0 0>, min=<0 0>}
    }
   }
   slot {class=OverlaySlot, properties {verticalAlignment=VAlign_Bottom, padding=<ltrb 5 0 0 0>}}
  }
  Image_Swatch {
   widgetClassPath=&Image
   bIsVariable=true
   parent=Overlay_98
   widget=Image_Swatch
   properties {
    visibility=SelfHitTestInvisible
    brush {
     bIsDynamicallyLoaded=false
     drawAs=Image
     imageType=NoImage
     mirroring=NoMirror
     resourceName=None
     resourceObject=&T_Icon_ColorPicker_swatch
     tiling=NoTile
     imageSize=<61 31>
     margin=<ltrb 0 0 0 0>
     outlineSettings {
      bUseBrushTransparency=false
      roundingType=HalfHeightRadius
      width=0
      color {colorUseRule=UseColor_Specified, specifiedColor=<rgba 0 0 0 0>}
      cornerRadii=<0 0 0 0>
     }
     tintColor {colorUseRule=UseColor_Specified, specifiedColor=<rgba 1 1 1 1>}
     uVRegion {bIsValid=false, max=<0 0>, min=<0 0>}
    }
   }
   slot {class=OverlaySlot, properties {verticalAlignment=VAlign_Bottom}}
  }
 ]
}
