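# List-entry button for the Creative property editor's color picker, wrapping one CreativeColorPickerSwatch that represents a color option.
# Setting the list item stores its ColorOption and syncs the swatch's color and selection.
# Hover and selection animate raising; selection also takes focus, and clicking while selected broadcasts OnSwatchCommitted.
# Is Represented Color Object matches options by color, icon, secondary value and names, ignoring gameplay tags.
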WidgetBlueprint /Game/Athena/UI/InGame/Creative/PropertyEditor/Color_Picker/CreativeColorPickerSwatchButton
graphs [
 EventGraph {nodes [
   n1 = Event OnListItemObjectSet()
   > n2 = Cast To ColorOptionObjectWrapper!(execute, Object=@n1.ListItemObject) -> (then, CastFailed, AsColor Option Object Wrapper:Color Option Object Wrapper Object Reference) {is=title K2Node_DynamicCast}
   > n3 = Set RepresentedSwatch(RepresentedSwatch:Playlist Option Tinted Icon Int Value Structure=Get ColorOption!(self:Color Option Object Wrapper Object Reference=@n2) -> (ColorOption:Playlist Option Tinted Icon Int Value Structure)) -> (Output_Get:Playlist Option Tinted Icon Int Value Structure)
   n4 = Get CreativeColorPickerSwatch() -> (CreativeColorPickerSwatch:Creative Color Picker Swatch Object Reference)
   > SetColorToRepresent!(execute, self:Creative Color Picker Swatch Object Reference=@n4, ColorOption:Playlist Option Tinted Icon Int Value Structure=@n3) -> (then)
   > n5 = Set bIsSelected(bIsSelected:bool=IsListItemSelected!(UserListEntry:User List Entry Interface) -> (ReturnValue:bool)) -> (Output_Get:bool)
   > SetRaised!(execute, self:Creative Color Picker Swatch Object Reference=@n4, bIsRaised:bool=@n5, bPlayAnimation:bool=false) -> (then)
   > SetSelected!(execute, self:Creative Color Picker Swatch Object Reference=@n4, bIsSelected:bool=@n5) -> (then)
   Event On Hovered()
   n6 = Get CreativeColorPickerSwatch()
   > SetRaised(self=@n6, bIsRaised=true, bPlayAnimation=true)
   Event On Unhovered()
   > Branch(Condition=NOT Boolean(A=Get bIsSelected() -> (bIsSelected:bool)))
   > SetRaised(self=@n6, bIsRaised=false, bPlayAnimation=true)
   Event On Clicked()
   > Branch(Condition=Get bIsSelected())
   > Call On Swatch Committed!(execute:Exec, self:Creative Color Picker Swatch Button Object Reference) -> (then:Exec) {is=title K2Node_CallDelegate}
   Event On Selected()
   > SetFocus()
   > n7 = Set bIsSelected(bIsSelected=true)
   n8 = Set bIsSelected(execute=@n11.then, bIsSelected=false)
   n9 = Merge Execution with Data!(Option A:Exec=@n7.then, Option A Data:bool=@n7, Option B:Exec=@n8.then, Option B Data:bool=@n8) -> (Then:Exec, Data:bool=~) {is=title K2Node_MacroInstance}
   n10 = Get CreativeColorPickerSwatch()
   > SetRaised(self=@n10, bIsRaised=@n9, bPlayAnimation=true)
   > SetSelected(self=@n10, bIsSelected=@n9)
   n11 = Event On Deselected()
  ]}
 Is Represented Color Object {nodes [
   n1 = Is Represented Color Object!() -> (then, InColorObject:Color Option Object Wrapper Object Reference) {is=title K2Node_FunctionEntry}
   > n2 = Is Valid(InputObject=@n1)
   Return Node!(execute=@n2.Is Not Valid, IsObject:bool=false)
   n3 = Get!() -> (
    RepresentedSwatch_Value:Linear Color Structure
    RepresentedSwatch_Icon:Texture 2D Object Reference
    RepresentedSwatch_SecondaryValue:int
    RepresentedSwatch_DisplayName:Text
    RepresentedSwatch_OptionValueName:String
    RepresentedSwatch_GrantedContextTags:Gameplay Tag Container Structure
   )
   n4 = Get!(self:Color Option Object Wrapper Object Reference=@n1) -> (
    ColorOption_Value:Linear Color Structure
    ColorOption_Icon:Texture 2D Object Reference
    ColorOption_SecondaryValue:int
    ColorOption_DisplayName:Text
    ColorOption_OptionValueName:String
    ColorOption_GrantedContextTags:Gameplay Tag Container Structure
   )
   > Return Node(IsObject=AND Boolean!(
     A="Equal (LinearColor)"!(A:Linear Color Structure=@n4, B:Linear Color Structure=@n3) -> (ReturnValue:bool)
     B="Equal (Object)"(A=@n4.ColorOption_Icon, B=@n3.RepresentedSwatch_Icon) {is=title K2Node_CallFunction}
     C:bool="Equal (Integer)"(A=@n4.ColorOption_SecondaryValue, B=@n3.RepresentedSwatch_SecondaryValue) {is=title K2Node_CallFunction}
     D:bool="Equal Exactly (Text)"!(A:"Text (by ref)"=@n4.ColorOption_DisplayName, B:"Text (by ref)"=@n3.RepresentedSwatch_DisplayName) -> (ReturnValue:bool)
     E:bool="Equal Exactly (String)"(A=@n4.ColorOption_OptionValueName, B=@n3.RepresentedSwatch_OptionValueName)
    ) -> (ReturnValue))
  ]}
 OnSwatchCommitted {nodes [OnSwatchCommitted() {is=title K2Node_FunctionEntry}]}
]
variables [
 bIsSelected {type=boolean}
 OnSwatchCommitted {}
 RepresentedSwatch {
  type=PlaylistOptionTintedIconIntValue
  default {
   displayName=""
   icon=&/T_Icon_ColorPicker_swatch
   optionValueName=""
   secondaryValue=0
   grantedContextTags {gameplayTags []}
   optionalDisplayIcon {largeIcon=null, smallIcon=null}
   value=<rgba 1 1 1 1>
  }
 }
]
widgets {
 info {
  inheritedWidgetCount=1
  parentClass=&CommonButtonBase
  rootWidgetClass=&/CreativeColorPickerSwatch.CreativeColorPickerSwatch_C
  widgetCount=2
 }
 widgets [
  CreativeColorPickerSwatch {
   widgetClassPath=&/CreativeColorPickerSwatch
   bIsVariable=true
   widget=CreativeColorPickerSwatch
   properties {visibility=HitTestInvisible}
  }
  InputActionWidget {widgetClassPath=&CommonActionWidget, bInherited=true, widget=null}
 ]
}
